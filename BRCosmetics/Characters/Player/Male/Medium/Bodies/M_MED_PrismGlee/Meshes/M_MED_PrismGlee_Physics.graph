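# Physics asset for the M_MED_PrismGlee male player body cosmetic, set up for physics-driven motion of its front and rear loincloth.
# Six capsule chains hang from tiny root bodies: left, center and right strips, with four segments in front and three in the rear.
# Constraints link each segment to the one above with narrow swing and twist limits, tighter on the rear.
# Pelvis and thigh capsules have no constraints and provide leg and hip collision shapes.
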PhysicsAsset /BRCosmetics/Characters/Player/Male/Medium/Bodies/M_MED_PrismGlee/Meshes/M_MED_PrismGlee_Physics
bodies {
 dyn_loincloth_front_01 [{shapeType=Capsule, length=6, radius=4, shapeName=dyn_loincloth_front_01_capsule, center=<4 6.4997 0>, rotation=<rot 90 0 0>}]
 dyn_loincloth_front_01_l [{shapeType=Capsule, length=6, radius=4, shapeName=dyn_loincloth_front_01_l_capsule, center=<4 2 0>, rotation=<rot 90 0 0>}]
 dyn_loincloth_front_01_r [{shapeType=Capsule, length=6, radius=4, shapeName=dyn_loincloth_front_01_r_capsule, center=<-4 -2 0>, rotation=<rot 90 180 0>}]
 dyn_loincloth_front_02 [{shapeType=Capsule, length=6, radius=4, shapeName=dyn_loincloth_front_02_capsule, center=<4 0 0>, rotation=<rot 90 0 0>}]
 dyn_loincloth_front_02_l [{shapeType=Capsule, length=6, radius=4, shapeName=dyn_loincloth_front_02_l_capsule, center=<4 1.7242 1.2724>, rotation=<rot 90 0 0>}]
 dyn_loincloth_front_02_r [{shapeType=Capsule, length=6, radius=4, shapeName=dyn_loincloth_front_02_r_capsule, center=<-4 -3.8308 0>, rotation=<rot 90 180 0>}]
 dyn_loincloth_front_03 [{shapeType=Capsule, length=6, radius=4, shapeName=dyn_loincloth_front_03_capsule, center=<4 0 0>, rotation=<rot 90 0 0>}]
 dyn_loincloth_front_03_l [{shapeType=Capsule, length=6, radius=4, shapeName=dyn_loincloth_front_03_l_capsule, center=<4.0042 1.6 1.6>, rotation=<rot 90 0 0>}]
 dyn_loincloth_front_03_r [{shapeType=Capsule, length=6, radius=4, shapeName=dyn_loincloth_front_03_r_capsule, center=<-4 -1.6053 -0.3259>, rotation=<rot 90 180 0>}]
 dyn_loincloth_front_04 [{shapeType=Capsule, length=6, radius=4, shapeName=dyn_loincloth_front_04_capsule, center=<4 0 0>, rotation=<rot 90 0 0>}]
 dyn_loincloth_front_04_l [{shapeType=Capsule, length=6, radius=4, shapeName=dyn_loincloth_front_04_l_capsule, center=<4 1.3607 0>, rotation=<rot 90 0 0>}]
 dyn_loincloth_front_04_r [{shapeType=Capsule, length=6, radius=4, shapeName=dyn_loincloth_front_04_r_capsule, center=<-4 -1.3108 0>, rotation=<rot 90 180 0>}]
 dyn_loincloth_front_l_root [{shapeType=Capsule, length=0.505, radius=0.505, shapeName=dyn_loincloth_front_l_root_capsule, rotation=<rot 0 -90 0>}]
 dyn_loincloth_front_r_root [{shapeType=Capsule, length=0.505, radius=0.505, shapeName=dyn_loincloth_front_r_root_capsule, rotation=<rot 0 -90 180>}]
 dyn_loincloth_front_root [{shapeType=Capsule, length=0.505, radius=0.505, shapeName=dyn_loincloth_front_root_capsule, rotation=<rot 0 -90 0>}]
 dyn_loincloth_rear_01 [{shapeType=Capsule, length=6, radius=4, shapeName=dyn_loincloth_rear_01_capsule, center=<4 2.0172 0>, rotation=<rot 90 0 0>}]
 dyn_loincloth_rear_01_l [{shapeType=Capsule, length=6, radius=4, shapeName=dyn_loincloth_rear_01_l_capsule, center=<4 2.6471 0>, rotation=<rot 90 0 0>}]
 dyn_loincloth_rear_01_r [{shapeType=Capsule, length=6, radius=4, shapeName=dyn_loincloth_rear_01_r_capsule, center=<-4 -2.6471 0>, rotation=<rot 90 180 0>}]
 dyn_loincloth_rear_02 [{shapeType=Capsule, length=6, radius=4, shapeName=dyn_loincloth_rear_02_capsule, center=<4 0 0>, rotation=<rot 90 0 0>}]
 dyn_loincloth_rear_02_l [{shapeType=Capsule, length=6, radius=4, shapeName=dyn_loincloth_rear_02_l_capsule, center=<4 0.5 0>, rotation=<rot 90 0 0>}]
 dyn_loincloth_rear_02_r [{shapeType=Capsule, length=6, radius=4, shapeName=dyn_loincloth_rear_02_r_capsule, center=<-4 -0.5 0>, rotation=<rot 90 180 0>}]
 dyn_loincloth_rear_03 [{shapeType=Capsule, length=6, radius=4, shapeName=dyn_loincloth_rear_03_capsule, center=<4 0 0>, rotation=<rot 90 0 0>}]
 dyn_loincloth_rear_03_l [{shapeType=Capsule, length=6, radius=4, shapeName=dyn_loincloth_rear_03_l_capsule, center=<4 0 0>, rotation=<rot 90 0 0>}]
 dyn_loincloth_rear_03_r [{shapeType=Capsule, length=6, radius=4, shapeName=dyn_loincloth_rear_03_r_capsule, center=<-4 0 0>, rotation=<rot 90 180 0>}]
 dyn_loincloth_rear_root [{shapeType=Capsule, length=0.505, radius=0.505, shapeName=dyn_loincloth_rear_root_capsule, rotation=<rot 0 -90 0>}]
 dyn_loincloth_rear_root_l [{shapeType=Capsule, length=0.505, radius=0.505, shapeName=dyn_loincloth_rear_root_l_capsule, rotation=<rot 0 -90 0>}]
 dyn_loincloth_rear_root_r [{shapeType=Capsule, length=0.505, radius=0.505, shapeName=dyn_loincloth_rear_root_r_capsule, rotation=<rot 0 -90 0>}]
 pelvis [{shapeType=Capsule, length=9.7796, radius=12}]
 thigh_l [
  {shapeType=Capsule, length=22.657, radius=8.4531, center=<-10.5461 -0.5061 -0.0853>, rotation=<rot -0.1497 -93.0522 -92.0548>}
  {
   shapeType=Capsule
   length=8.7042
   radius=8.1092
   shapeName=thigh_l_capsule
   center=<-30.0625 0.1958 0.3761>
   rotation=<rot 89.1032 -29.8905 -29.4514>
  }
 ]
 thigh_r [
  {shapeType=Capsule, length=22.657, radius=8.4531, center=<10.5461 0.5061 0.0853>, rotation=<rot -0.1497 -93.0522 87.9452>}
  {
   shapeType=Capsule
   length=8.7042
   radius=8.1092
   shapeName=thigh_r_capsule
   center=<30.0625 -0.1958 -0.3761>
   rotation=<rot 89.1032 -29.8905 150.5486>
  }
 ]
}
constraints [
 {bone1Name=dyn_loincloth_front_02_r, bone2Name=dyn_loincloth_front_01_r, swing1LimitDegrees=14, swing2LimitDegrees=14, twistLimitDegrees=4}
 {bone1Name=dyn_loincloth_front_03_r, bone2Name=dyn_loincloth_front_02_r, swing1LimitDegrees=14, swing2LimitDegrees=14, twistLimitDegrees=4}
 {bone1Name=dyn_loincloth_front_04_r, bone2Name=dyn_loincloth_front_03_r, swing1LimitDegrees=14, swing2LimitDegrees=14, twistLimitDegrees=4}
 {bone1Name=dyn_loincloth_front_01_l, bone2Name=dyn_loincloth_front_l_root, swing1LimitDegrees=14, swing2LimitDegrees=14, twistLimitDegrees=4}
 {bone1Name=dyn_loincloth_front_02_l, bone2Name=dyn_loincloth_front_01_l, swing1LimitDegrees=14, swing2LimitDegrees=14, twistLimitDegrees=4}
 {bone1Name=dyn_loincloth_front_03_l, bone2Name=dyn_loincloth_front_02_l, swing1LimitDegrees=14, swing2LimitDegrees=14, twistLimitDegrees=4}
 {bone1Name=dyn_loincloth_front_04_l, bone2Name=dyn_loincloth_front_03_l, swing1LimitDegrees=14, swing2LimitDegrees=14, twistLimitDegrees=4}
 {bone1Name=dyn_loincloth_front_01, bone2Name=dyn_loincloth_front_root, swing1LimitDegrees=14, swing2LimitDegrees=14, twistLimitDegrees=4}
 {bone1Name=dyn_loincloth_front_02, bone2Name=dyn_loincloth_front_01, swing1LimitDegrees=14, swing2LimitDegrees=14, twistLimitDegrees=4}
 {bone1Name=dyn_loincloth_front_03, bone2Name=dyn_loincloth_front_02, swing1LimitDegrees=14, swing2LimitDegrees=14, twistLimitDegrees=4}
 {bone1Name=dyn_loincloth_front_04, bone2Name=dyn_loincloth_front_03, swing1LimitDegrees=14, swing2LimitDegrees=14, twistLimitDegrees=4}
 {bone1Name=dyn_loincloth_rear_01, bone2Name=dyn_loincloth_rear_root, swing1LimitDegrees=12, swing2LimitDegrees=12, twistLimitDegrees=2}
 {bone1Name=dyn_loincloth_rear_02, bone2Name=dyn_loincloth_rear_01, swing1LimitDegrees=12, swing2LimitDegrees=12, twistLimitDegrees=2}
 {bone1Name=dyn_loincloth_rear_03, bone2Name=dyn_loincloth_rear_02, swing1LimitDegrees=12, swing2LimitDegrees=12, twistLimitDegrees=2}
 {bone1Name=dyn_loincloth_rear_01_l, bone2Name=dyn_loincloth_rear_root_l, swing1LimitDegrees=12, swing2LimitDegrees=12, twistLimitDegrees=2}
 {bone1Name=dyn_loincloth_rear_02_l, bone2Name=dyn_loincloth_rear_01_l, swing1LimitDegrees=12, swing2LimitDegrees=12, twistLimitDegrees=2}
 {bone1Name=dyn_loincloth_rear_03_l, bone2Name=dyn_loincloth_rear_02_l, swing1LimitDegrees=12, swing2LimitDegrees=12, twistLimitDegrees=2}
 {bone1Name=dyn_loincloth_rear_01_r, bone2Name=dyn_loincloth_rear_root_r, swing1LimitDegrees=12, swing2LimitDegrees=12, twistLimitDegrees=2}
 {bone1Name=dyn_loincloth_rear_02_r, bone2Name=dyn_loincloth_rear_01_r, swing1LimitDegrees=12, swing2LimitDegrees=12, twistLimitDegrees=2}
 {bone1Name=dyn_loincloth_rear_03_r, bone2Name=dyn_loincloth_rear_02_r, swing1LimitDegrees=12, swing2LimitDegrees=12, twistLimitDegrees=2}
 {bone1Name=dyn_loincloth_front_01_r, bone2Name=dyn_loincloth_front_r_root, swing1LimitDegrees=14, swing2LimitDegrees=14, twistLimitDegrees=4}
]
