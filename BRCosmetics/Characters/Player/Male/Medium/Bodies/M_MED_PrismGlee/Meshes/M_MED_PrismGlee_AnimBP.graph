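# Post-process style AnimBP for the PrismGlee male body cosmetic, adding secondary motion to the pose copied from the parent mesh.
# A velocity-fed Control Rig leads into pelvis-space RigidBody and shoulder-strap AnimDynamics.
# bPbBlendEnabled optionally adds muscle jiggle on biceps, triceps, pecs and quads.
# The owner's PrismGlee_Emote curve layers DefaultSlot onto front loincloth bones, while skydiving events set a Gravity Override variable.
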AnimBlueprint /BRCosmetics/Characters/Player/Male/Medium/Bodies/M_MED_PrismGlee/Meshes/M_MED_PrismGlee_AnimBP
target_skeleton=&/M_MED_PrismGlee_Skeleton
graphs [
 AnimGraph {nodes [
   n1 = Layered blend per bone(
    BasePose=Blend Poses by bool(
     BlendPose_0=Use cached pose 'Post Body Dynamics'()
     BlendPose_1=Use cached pose 'Post Physics'()
     BlendTime_0=0
     BlendTime_1=0
     bActiveValue=Get bPbBlendEnabled() -> (bPbBlendEnabled:bool)
    )
    BlendPoses_0=Slot 'DefaultSlot'()
    BlendWeights_0=Get EmoteCurve() -> (EmoteCurve:float)
   ) {layerSetup [{branchFilters [
       {blendDepth=1, boneName=dyn_loincloth_front_l_root}
       {blendDepth=1, boneName=dyn_loincloth_front_r_root}
       {blendDepth=1, boneName=dyn_loincloth_front_root}
      ]}]}
   Output Pose(Result=@n1) {layerGroup=None}
   n2 = Control Rig!(Source=Use cached pose 'Initial_Pose_1'(), Alpha, Player Speed Z:double=Get PawnVelocityZ()) -> (Pose) {controlRigAssetReference {blueprintRigClass=&/M_MED_PrismGlee_CtrlRig.M_MED_PrismGlee_CtrlRig_C}}
   Save cached pose 'Post Deformation'(Pose=@n2) {cacheName=Post Deformation}
   n3 = RigidBody!(ComponentPose=Local To Component(LocalPose=Use cached pose 'Post Deformation'()), Alpha, OverrideWorldGravity:Vector="0,0,0", ExternalForce) -> (Pose) {
    bOverrideWorldGravity=true
    overridePhysicsAsset=&/M_MED_PrismGlee_Physics
    simulationSpace=BaseBoneSpace
    baseBoneRef {boneName=pelvis}
    simSpaceSettings {maxAngularAcceleration=5, maxAngularVelocity=15, maxLinearAcceleration=8, maxLinearVelocity=1400, worldAlpha=1, externalLinearDragV=<3 1 1>}
   }
   n4 = Anim Dynamics - Bone: dyn_shoulder_straps_01_l(ComponentPose=@n3, Alpha=1, ExternalForce="0,0,0") {
    angularDampingOverride=1
    angularSpringConstant=10000
    bAngularSpring=true
    bOverrideAngularDamping=true
    bUseGravityOverride=true
    gravityScale=0.35
    simulationSpace=RootRelative
    boundBone {boneName=dyn_shoulder_straps_01_l}
    componentAppliedLinearAccClamp=<0 0 0>
    gravityOverride {z=-2200}
    physicsBodyDefinitions [{
      boundBone {boneName=dyn_shoulder_straps_01_l}
      boxExtents=<25 30 30>
      constraintSetup {angularLimitsMax=<10 10 10>, angularLimitsMin=<-10 -10 -10>, angularTarget=<1 0 0>}
      localJointOffset=<2.3 3 -1>
     }]
    relativeSpaceBone {boneName=spine_05}
    simSpaceSettings {maxAngularAcceleration=15, maxAngularVelocity=450, simSpaceAngularAlpha=1}
   }
   n5 = Anim Dynamics - Bone: dyn_shoulder_straps_02_l(ComponentPose=@n4, Alpha=1, ExternalForce="0,0,0") {
    angularDampingOverride=1
    angularSpringConstant=15000
    bAngularSpring=true
    bOverrideAngularDamping=true
    bUseGravityOverride=true
    gravityScale=0.35
    simulationSpace=RootRelative
    boundBone {boneName=dyn_shoulder_straps_02_l}
    componentAppliedLinearAccClamp=<0 0 0>
    gravityOverride {z=-1200}
    physicsBodyDefinitions [{
      boundBone {boneName=dyn_shoulder_straps_02_l}
      boxExtents=<20 30 30>
      constraintSetup {angularLimitsMax=<10 20 10>, angularLimitsMin=<-10 -10 -10>, angularTarget=<1 0 0>}
      localJointOffset=<2 2 0>
     }]
    relativeSpaceBone {boneName=spine_05}
    simSpaceSettings {maxAngularAcceleration=15, maxAngularVelocity=450, simSpaceAngularAlpha=1}
   }
   n6 = Anim Dynamics - Bone: dyn_shoulder_straps_03_l(ComponentPose=@n5, Alpha=0.8, ExternalForce="0,0,0") {
    angularDampingOverride=1
    angularSpringConstant=20000
    bAngularSpring=true
    bOverrideAngularDamping=true
    bUseGravityOverride=true
    gravityScale=0.35
    simulationSpace=RootRelative
    boundBone {boneName=dyn_shoulder_straps_03_l}
    componentAppliedLinearAccClamp=<0 0 0>
    gravityOverride {z=-2200}
    physicsBodyDefinitions [{
      boundBone {boneName=dyn_shoulder_straps_03_l}
      boxExtents=<20 30 30>
      constraintSetup {angularLimitsMax=<0 0 5>, angularLimitsMin=<0 -5 -5>, angularTarget=<1 0 0>}
      localJointOffset=<2 0 2>
     }]
    relativeSpaceBone {boneName=spine_05}
    simSpaceSettings {maxAngularAcceleration=15, maxAngularVelocity=450, simSpaceAngularAlpha=1}
   }
   n7 = Anim Dynamics - Bone: dyn_shoulder_straps_01_r(ComponentPose=@n6, Alpha=1, ExternalForce="0,0,0") {
    angularDampingOverride=1
    angularSpringConstant=10000
    bAngularSpring=true
    bOverrideAngularDamping=true
    bUseGravityOverride=true
    gravityScale=0.35
    simulationSpace=RootRelative
    boundBone {boneName=dyn_shoulder_straps_01_r}
    componentAppliedLinearAccClamp=<0 0 0>
    gravityOverride {z=-2200}
    physicsBodyDefinitions [{
      boundBone {boneName=dyn_shoulder_straps_01_r}
      boxExtents=<20 35 30>
      constraintSetup {angularLimitsMax=<10 10 10>, angularLimitsMin=<-10 -10 -10>, angularTarget=<1 0 0>}
      localJointOffset=<-2 3 1>
     }]
    relativeSpaceBone {boneName=spine_05}
    simSpaceSettings {maxAngularAcceleration=15, maxAngularVelocity=450, simSpaceAngularAlpha=1}
   }
   n8 = Anim Dynamics - Bone: dyn_shoulder_straps_02_r(ComponentPose=@n7, Alpha=1, ExternalForce="0,0,0") {
    angularDampingOverride=1
    angularSpringConstant=15000
    bAngularSpring=true
    bOverrideAngularDamping=true
    bUseGravityOverride=true
    gravityScale=0.35
    simulationSpace=RootRelative
    boundBone {boneName=dyn_shoulder_straps_02_r}
    componentAppliedLinearAccClamp=<0 0 0>
    gravityOverride {z=-1200}
    physicsBodyDefinitions [{
      boundBone {boneName=dyn_shoulder_straps_02_r}
      boxExtents=<20 30 30>
      constraintSetup {angularLimitsMax=<10 20 10>, angularLimitsMin=<-10 -10 -10>, angularTarget=<1 0 0>}
      localJointOffset=<-2 1 0>
     }]
    relativeSpaceBone {boneName=spine_05}
    simSpaceSettings {maxAngularAcceleration=15, maxAngularVelocity=450, simSpaceAngularAlpha=1}
   }
   n9 = Anim Dynamics - Bone: dyn_shoulder_straps_03_r(ComponentPose=@n8, Alpha=0.8, ExternalForce="0,0,0") {
    angularDampingOverride=1
    angularSpringConstant=20000
    bAngularSpring=true
    bOverrideAngularDamping=true
    bUseGravityOverride=true
    gravityScale=0.35
    simulationSpace=RootRelative
    boundBone {boneName=dyn_shoulder_straps_03_r}
    componentAppliedLinearAccClamp=<0 0 0>
    gravityOverride {z=-2200}
    physicsBodyDefinitions [{
      boundBone {boneName=dyn_shoulder_straps_03_r}
      boxExtents=<20 30 30>
      constraintSetup {angularLimitsMax=<0 0 5>, angularLimitsMin=<0 -5 -5>, angularTarget=<1 0 0>}
      localJointOffset=<-2 -2 -3>
     }]
    relativeSpaceBone {boneName=spine_05}
    simSpaceSettings {maxAngularAcceleration=15, maxAngularVelocity=450, simSpaceAngularAlpha=1}
   }
   Save cached pose 'Post Physics'(Pose=Component To Local(ComponentPose=@n9)) {cacheName=Post Physics}
   n10 = Anim Dynamics - Bone: deform_bicep_01_l!(ComponentPose=Local To Component(LocalPose=Use cached pose 'Post Physics'()), ExternalForce="0,0,0") -> (Pose) {
    alpha=0.4
    angularSpringConstant=15000
    bLinearSpring=true
    bOverrideLinearDamping=true
    bUseGravityOverride=true
    bUsePlanarLimit=false
    linearDampingOverride=0.9999
    linearSpringConstant=300
    simulationSpace=RootRelative
    boundBone {boneName=deform_bicep_01_l}
    componentAppliedLinearAccClamp=<1000 1000 1000>
    physicsBodyDefinitions [{
      boundBone {boneName=deform_bicep_01_l}
      boxExtents=<38 38 38>
      constraintSetup {angularLimitsMax=<0 0 0>, angularLimitsMin=<0 0 0>, angularTarget=<1 0 0>, linearAxesMax=<0 0 1>, linearAxesMin=<0 0 -1>}
      localJointOffset=<-2 0 0>
     }]
    relativeSpaceBone {boneName=pelvis}
    simSpaceSettings {maxAngularAcceleration=40, maxAngularVelocity=770, simSpaceAngularAlpha=0.4}
   }
   n11 = Anim Dynamics - Bone: deform_bicep_01_l(ComponentPose=@n10, ExternalForce="0,0,0") {
    alpha=0.4
    angularDampingOverride=0.9999
    angularSpringConstant=15000
    bAngularSpring=true
    bOverrideAngularDamping=true
    bUseGravityOverride=true
    bUsePlanarLimit=false
    simulationSpace=RootRelative
    boundBone {boneName=deform_bicep_01_l}
    componentAppliedLinearAccClamp=<1000 1000 1000>
    gravityOverride {z=-980}
    physicsBodyDefinitions [{
      boundBone {boneName=deform_bicep_01_l}
      boxExtents=<24 38 24>
      constraintSetup {angularLimitsMax=<1 10 1>, angularLimitsMin=<-1 -10 -1>, angularTarget=<1 0 0>}
      localJointOffset=<-5 0 0>
     }]
    relativeSpaceBone {boneName=pelvis}
    simSpaceSettings {maxAngularAcceleration=10, maxAngularVelocity=300, simSpaceAngularAlpha=0.4}
   }
   n12 = Anim Dynamics - Bone: deform_bicep_01_r!(ComponentPose=@n11, ExternalForce="0,0,0") -> (Pose) {
    alpha=0.4
    angularSpringConstant=15000
    bLinearSpring=true
    bOverrideLinearDamping=true
    bUseGravityOverride=true
    bUsePlanarLimit=false
    linearDampingOverride=0.9999
    linearSpringConstant=300
    simulationSpace=RootRelative
    boundBone {boneName=deform_bicep_01_r}
    componentAppliedLinearAccClamp=<1000 1000 1000>
    physicsBodyDefinitions [{
      boundBone {boneName=deform_bicep_01_r}
      boxExtents=<38 38 38>
      constraintSetup {angularLimitsMax=<0 0 0>, angularLimitsMin=<0 0 0>, angularTarget=<1 0 0>, linearAxesMax=<0 0 1>, linearAxesMin=<0 0 -1>}
      localJointOffset=<2 0 0>
     }]
    relativeSpaceBone {boneName=pelvis}
    simSpaceSettings {maxAngularAcceleration=40, maxAngularVelocity=770, simSpaceAngularAlpha=0.4}
   }
   n13 = Anim Dynamics - Bone: deform_bicep_01_r(ComponentPose=@n12, ExternalForce="0,0,0") {
    alpha=0.4
    angularDampingOverride=0.9999
    angularSpringConstant=15000
    bAngularSpring=true
    bOverrideAngularDamping=true
    bUseGravityOverride=true
    bUsePlanarLimit=false
    simulationSpace=RootRelative
    boundBone {boneName=deform_bicep_01_r}
    componentAppliedLinearAccClamp=<1000 1000 1000>
    gravityOverride {z=-980}
    physicsBodyDefinitions [{
      boundBone {boneName=deform_bicep_01_r}
      boxExtents=<24 38 24>
      constraintSetup {angularLimitsMax=<1 10 1>, angularLimitsMin=<-1 -10 -1>, angularTarget=<1 0 0>}
      localJointOffset=<-5 0 0>
     }]
    relativeSpaceBone {boneName=pelvis}
    simSpaceSettings {maxAngularAcceleration=10, maxAngularVelocity=300, simSpaceAngularAlpha=0.4}
   }
   n14 = Anim Dynamics - Bone: deform_tricep_01_l!(ComponentPose=@n13, ExternalForce="0,0,0") -> (Pose) {
    alpha=0.4
    angularSpringConstant=15000
    bLinearSpring=true
    bOverrideLinearDamping=true
    bUseGravityOverride=true
    bUsePlanarLimit=false
    linearDampingOverride=0.9999
    linearSpringConstant=200
    simulationSpace=RootRelative
    boundBone {boneName=deform_tricep_01_l}
    componentAppliedLinearAccClamp=<1000 1000 1000>
    gravityOverride {z=-100}
    physicsBodyDefinitions [{
      boundBone {boneName=deform_tricep_01_l}
      boxExtents=<48 58 48>
      constraintSetup {angularLimitsMax=<0 0 0>, angularLimitsMin=<0 0 0>, angularTarget=<1 0 0>, linearAxesMax=<0 0.1 1.5>, linearAxesMin=<0 -0.1 -1.5>}
      localJointOffset=<-3 0 0>
     }]
    relativeSpaceBone {boneName=pelvis}
    simSpaceSettings {maxAngularAcceleration=30, maxAngularVelocity=300, simSpaceAngularAlpha=0.4}
   }
   n15 = Anim Dynamics - Bone: deform_tricep_01_l(ComponentPose=@n14, ExternalForce="0,0,0") {
    alpha=0.4
    angularDampingOverride=0.9999
    angularSpringConstant=50000
    bAngularSpring=true
    bOverrideAngularDamping=true
    bUseGravityOverride=true
    bUsePlanarLimit=false
    simulationSpace=RootRelative
    boundBone {boneName=deform_tricep_01_l}
    componentAppliedLinearAccClamp=<1000 1000 1000>
    gravityOverride {z=-980}
    physicsBodyDefinitions [{
      boundBone {boneName=deform_tricep_01_l}
      boxExtents=<24 38 24>
      constraintSetup {angularLimitsMax=<1 10 1>, angularLimitsMin=<-1 -10 -1>, angularTarget=<1 0 0>}
      localJointOffset=<5 0 0>
     }]
    relativeSpaceBone {boneName=pelvis}
    simSpaceSettings {maxAngularAcceleration=10, maxAngularVelocity=300, simSpaceAngularAlpha=0.4}
   }
   n16 = Anim Dynamics - Bone: deform_tricep_01_r!(ComponentPose=@n15, ExternalForce="0,0,0") -> (Pose) {
    alpha=0.4
    angularSpringConstant=15000
    bLinearSpring=true
    bOverrideLinearDamping=true
    bUseGravityOverride=true
    bUsePlanarLimit=false
    linearDampingOverride=0.9999
    linearSpringConstant=200
    simulationSpace=RootRelative
    boundBone {boneName=deform_tricep_01_r}
    componentAppliedLinearAccClamp=<1000 1000 1000>
    gravityOverride {z=-100}
    physicsBodyDefinitions [{
      boundBone {boneName=deform_tricep_01_r}
      boxExtents=<48 58 48>
      constraintSetup {angularLimitsMax=<0 0 0>, angularLimitsMin=<0 0 0>, angularTarget=<1 0 0>, linearAxesMax=<0 0.1 1.5>, linearAxesMin=<0 -0.1 -1.5>}
      localJointOffset=<3 0 0>
     }]
    relativeSpaceBone {boneName=pelvis}
    simSpaceSettings {maxAngularAcceleration=30, maxAngularVelocity=300, simSpaceAngularAlpha=0.4}
   }
   n17 = Anim Dynamics - Bone: deform_tricep_01_r(ComponentPose=@n16, ExternalForce="0,0,0") {
    alpha=0.4
    angularDampingOverride=0.9999
    angularSpringConstant=50000
    bAngularSpring=true
    bOverrideAngularDamping=true
    bUseGravityOverride=true
    bUsePlanarLimit=false
    simulationSpace=RootRelative
    boundBone {boneName=deform_tricep_01_r}
    componentAppliedLinearAccClamp=<1000 1000 1000>
    gravityOverride {z=-980}
    physicsBodyDefinitions [{
      boundBone {boneName=deform_tricep_01_r}
      boxExtents=<24 38 24>
      constraintSetup {angularLimitsMax=<1 10 1>, angularLimitsMin=<-1 -10 -1>, angularTarget=<1 0 0>}
      localJointOffset=<5 0 0>
     }]
    relativeSpaceBone {boneName=pelvis}
    simSpaceSettings {maxAngularAcceleration=10, maxAngularVelocity=300, simSpaceAngularAlpha=0.4}
   }
   n18 = Anim Dynamics - Bone: deform_tricep_02_l!(ComponentPose=@n17, ExternalForce="0,0,0") -> (Pose) {
    alpha=0.4
    angularSpringConstant=15000
    bLinearSpring=true
    bOverrideLinearDamping=true
    bUseGravityOverride=true
    bUsePlanarLimit=false
    linearDampingOverride=0.9999
    linearSpringConstant=200
    simulationSpace=RootRelative
    boundBone {boneName=deform_tricep_02_l}
    componentAppliedLinearAccClamp=<1000 1000 1000>
    gravityOverride {z=-100}
    physicsBodyDefinitions [{
      boundBone {boneName=deform_tricep_02_l}
      boxExtents=<48 58 48>
      constraintSetup {angularLimitsMax=<0 0 0>, angularLimitsMin=<0 0 0>, angularTarget=<1 0 0>, linearAxesMax=<0 0.1 1.5>, linearAxesMin=<0 -0.1 -1.5>}
      localJointOffset=<-3 0 0>
     }]
    relativeSpaceBone {boneName=pelvis}
    simSpaceSettings {maxAngularAcceleration=30, maxAngularVelocity=300, simSpaceAngularAlpha=0.4}
   }
   n19 = Anim Dynamics - Bone: deform_tricep_02_l(ComponentPose=@n18, ExternalForce="0,0,0") {
    alpha=0.4
    angularDampingOverride=0.9999
    angularSpringConstant=50000
    bAngularSpring=true
    bOverrideAngularDamping=true
    bUseGravityOverride=true
    bUsePlanarLimit=false
    simulationSpace=RootRelative
    boundBone {boneName=deform_tricep_02_l}
    componentAppliedLinearAccClamp=<1000 1000 1000>
    gravityOverride {z=-980}
    physicsBodyDefinitions [{
      boundBone {boneName=deform_tricep_02_l}
      boxExtents=<24 38 24>
      constraintSetup {angularLimitsMax=<1 10 1>, angularLimitsMin=<-1 -10 -1>, angularTarget=<1 0 0>}
      localJointOffset=<5 0 0>
     }]
    relativeSpaceBone {boneName=pelvis}
    simSpaceSettings {maxAngularAcceleration=10, maxAngularVelocity=300, simSpaceAngularAlpha=0.4}
   }
   n20 = Anim Dynamics - Bone: deform_tricep_02_r!(ComponentPose=@n19, ExternalForce="0,0,0") -> (Pose) {
    alpha=0.4
    angularSpringConstant=15000
    bLinearSpring=true
    bOverrideLinearDamping=true
    bUseGravityOverride=true
    bUsePlanarLimit=false
    linearDampingOverride=0.9999
    linearSpringConstant=200
    simulationSpace=RootRelative
    boundBone {boneName=deform_tricep_02_r}
    componentAppliedLinearAccClamp=<1000 1000 1000>
    gravityOverride {z=-100}
    physicsBodyDefinitions [{
      boundBone {boneName=deform_tricep_02_r}
      boxExtents=<48 58 48>
      constraintSetup {angularLimitsMax=<0 0 0>, angularLimitsMin=<0 0 0>, angularTarget=<1 0 0>, linearAxesMax=<0 0.1 1.5>, linearAxesMin=<0 -0.1 -1.5>}
      localJointOffset=<3 0 0>
     }]
    relativeSpaceBone {boneName=pelvis}
    simSpaceSettings {maxAngularAcceleration=30, maxAngularVelocity=300, simSpaceAngularAlpha=0.4}
   }
   n21 = Anim Dynamics - Bone: deform_tricep_02_r(ComponentPose=@n20, ExternalForce="0,0,0") {
    alpha=0.4
    angularDampingOverride=0.9999
    angularSpringConstant=50000
    bAngularSpring=true
    bOverrideAngularDamping=true
    bUseGravityOverride=true
    bUsePlanarLimit=false
    simulationSpace=RootRelative
    boundBone {boneName=deform_tricep_02_r}
    componentAppliedLinearAccClamp=<1000 1000 1000>
    gravityOverride {z=-980}
    physicsBodyDefinitions [{
      boundBone {boneName=deform_tricep_02_r}
      boxExtents=<24 38 24>
      constraintSetup {angularLimitsMax=<1 10 1>, angularLimitsMin=<-1 -10 -1>, angularTarget=<1 0 0>}
      localJointOffset=<5 0 0>
     }]
    relativeSpaceBone {boneName=pelvis}
    simSpaceSettings {maxAngularAcceleration=10, maxAngularVelocity=300, simSpaceAngularAlpha=0.4}
   }
   n22 = Anim Dynamics - Bone: deform_pec_l(ComponentPose=@n21, Alpha=0.6, ExternalForce="0,0,0") {
    angularSpringConstant=15000
    bLinearSpring=true
    bOverrideLinearDamping=true
    bUseGravityOverride=true
    bUsePlanarLimit=false
    linearDampingOverride=1
    linearSpringConstant=300
    simulationSpace=RootRelative
    boundBone {boneName=deform_pec_l}
    gravityOverride {z=-200}
    physicsBodyDefinitions [{
      boundBone {boneName=deform_pec_l}
      boxExtents=<50 55 60>
      constraintSetup {angularLimitsMax=<0 0 0>, angularLimitsMin=<0 0 0>, angularTarget=<1 0 0>, linearAxesMax=<0.5 0 1.5>, linearAxesMin=<-0.5 0 -1.5>}
      localJointOffset=<-4 0 0>
     }]
    relativeSpaceBone {boneName=spine_04}
    simSpaceSettings {maxAngularAcceleration=10, maxAngularVelocity=770, simSpaceAngularAlpha=1}
   }
   n23 = Anim Dynamics - Bone: deform_pec_r(ComponentPose=@n22, Alpha=0.6, ExternalForce="0,0,0") {
    angularSpringConstant=15000
    bLinearSpring=true
    bOverrideLinearDamping=true
    bUseGravityOverride=true
    bUsePlanarLimit=false
    linearDampingOverride=1
    linearSpringConstant=300
    simulationSpace=RootRelative
    boundBone {boneName=deform_pec_r}
    gravityOverride {z=-200}
    physicsBodyDefinitions [{
      boundBone {boneName=deform_pec_r}
      boxExtents=<50 55 60>
      constraintSetup {angularLimitsMax=<0 0 0>, angularLimitsMin=<0 0 0>, angularTarget=<1 0 0>, linearAxesMax=<0.5 0 1.5>, linearAxesMin=<-0.5 0 -1.5>}
      localJointOffset=<4 0 0>
     }]
    relativeSpaceBone {boneName=spine_04}
    simSpaceSettings {maxAngularAcceleration=10, maxAngularVelocity=770, simSpaceAngularAlpha=1}
   }
   n24 = Anim Dynamics - Bone: dyn_quad_l(ComponentPose=@n23, Alpha=0.7, ExternalForce="0,0,0") {
    angularSpringConstant=15000
    bLinearSpring=true
    bOverrideLinearDamping=true
    bUseGravityOverride=true
    bUsePlanarLimit=false
    linearDampingOverride=0.999
    linearSpringConstant=350
    simulationSpace=RootRelative
    boundBone {boneName=dyn_quad_l}
    gravityOverride {z=-480}
    physicsBodyDefinitions [{
      boundBone {boneName=dyn_quad_l}
      boxExtents=<50 55 60>
      constraintSetup {angularLimitsMax=<0 0 0>, angularLimitsMin=<0 0 0>, angularTarget=<1 0 0>, linearAxesMax=<1 0.3 0.3>, linearAxesMin=<-1 -0.3 -0.3>}
      localJointOffset=<-4 0 0>
     }]
    relativeSpaceBone {boneName=spine_04}
    simSpaceSettings {maxAngularAcceleration=10, maxAngularVelocity=770, simSpaceAngularAlpha=1}
   }
   n25 = Anim Dynamics - Bone: dyn_quad_r(ComponentPose=@n24, Alpha=0.7, ExternalForce="0,0,0") {
    angularSpringConstant=15000
    bLinearSpring=true
    bOverrideLinearDamping=true
    bUseGravityOverride=true
    bUsePlanarLimit=false
    linearDampingOverride=0.999
    linearSpringConstant=350
    simulationSpace=RootRelative
    boundBone {boneName=dyn_quad_r}
    gravityOverride {z=-480}
    physicsBodyDefinitions [{
      boundBone {boneName=dyn_quad_r}
      boxExtents=<50 55 60>
      constraintSetup {angularLimitsMax=<0 0 0>, angularLimitsMin=<0 0 0>, angularTarget=<1 0 0>, linearAxesMax=<1 0.3 0.3>, linearAxesMin=<-1 -0.3 -0.3>}
      localJointOffset=<4 0 0>
     }]
    relativeSpaceBone {boneName=spine_04}
    simSpaceSettings {maxAngularAcceleration=10, maxAngularVelocity=770, simSpaceAngularAlpha=1}
   }
   n26 = Anim Dynamics - Bone: dyn_quad_l!(ComponentPose=@n25, ExternalForce="0,0,0") -> (Pose) {
    alpha=0.8
    angularDampingOverride=0.9999
    angularSpringConstant=50000
    bAngularSpring=true
    bOverrideAngularDamping=true
    bUseGravityOverride=true
    bUsePlanarLimit=false
    simulationSpace=RootRelative
    boundBone {boneName=dyn_quad_l}
    componentAppliedLinearAccClamp=<1000 1000 1000>
    gravityOverride {z=-980}
    physicsBodyDefinitions [{
      boundBone {boneName=dyn_quad_l}
      boxExtents=<38 24 24>
      constraintSetup {angularLimitsMax=<16 5 5>, angularLimitsMin=<-16 -5 -5>, angularTarget=<1 0 0>}
      localJointOffset=<-5 0 0>
     }]
    relativeSpaceBone {boneName=pelvis}
    simSpaceSettings {maxAngularAcceleration=10, maxAngularVelocity=300, simSpaceAngularAlpha=0.6}
   }
   n27 = Anim Dynamics - Bone: dyn_quad_r!(ComponentPose=@n26, ExternalForce="0,0,0") -> (Pose) {
    alpha=0.8
    angularDampingOverride=0.9999
    angularSpringConstant=50000
    bAngularSpring=true
    bOverrideAngularDamping=true
    bUseGravityOverride=true
    bUsePlanarLimit=false
    simulationSpace=RootRelative
    boundBone {boneName=dyn_quad_r}
    componentAppliedLinearAccClamp=<1000 1000 1000>
    gravityOverride {z=-980}
    physicsBodyDefinitions [{
      boundBone {boneName=dyn_quad_r}
      boxExtents=<38 24 24>
      constraintSetup {angularLimitsMax=<16 5 5>, angularLimitsMin=<-16 -5 -5>, angularTarget=<1 0 0>}
      localJointOffset=<5 0 0>
     }]
    relativeSpaceBone {boneName=pelvis}
    simSpaceSettings {maxAngularAcceleration=10, maxAngularVelocity=300, simSpaceAngularAlpha=0.6}
   }
   Save cached pose 'Post Body Dynamics'(Pose=Component To Local(ComponentPose=@n27)) {cacheName=Post Body Dynamics}
   Save cached pose 'Initial_Pose_1'(Pose=Copy Pose From Mesh() {bUseAttachedParent=true}) {cacheName=Initial_Pose_1}
   Use cached pose 'Initial_Pose_1'()
  ]}
 EventGraph {nodes [
   Event BlueprintUpdateAnimation()
   > Set EmoteCurve(EmoteCurve:float=GetCurveValue(
     self=GetOwnerAnimInstance!(self:Custom Character Part Anim Instance Object Reference) -> (ReturnValue:Anim Instance Object Reference)
     CurveName=PrismGlee_Emote
    )) -> (Output_Get:float)
   n1 = Event OnBeginSkydiving()
   Set Gravity Override(execute=[@n1.then, @n2.then], Gravity Override="0,0,-400")
   n2 = Event OnBeginSkydivingFromBus() {is=title K2Node_Event}
   Event OnEndSkydiving()
   > Set Gravity Override(Gravity Override="0,0,-980")
   "Gravity Override State Function (Parameter Asset) (Nativized)"(
    CharacterPartAnimInstance=Self-Reference()
    GameTimeInSeconds=GetGameTimeInSeconds()
    Parameters=&/M_MED_PrismGlee_Parameters
   )
   Set Gravity Override(Gravity Override="0, 0, 0")
  ]}
]
variables [
 EmoteCurve {type=number}
 Gravity Override {type=Vector, default=<0 0 -980>}
]
