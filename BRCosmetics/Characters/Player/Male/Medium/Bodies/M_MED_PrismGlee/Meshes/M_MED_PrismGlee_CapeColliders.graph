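# Physics asset that supplies cape collision shapes for the M_MED_PrismGlee male medium player body, as its CapeColliders name indicates.
# Capsules wrap the pelvis, spine, neck and clavicles, with extra capsules on the dynamic rear loincloth and shoulder strap bones.
# Arm, leg and foot bodies are present but carry no shapes.
# It has no constraints, so it defines collision volumes only, with no ragdoll joints.
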PhysicsAsset /BRCosmetics/Characters/Player/Male/Medium/Bodies/M_MED_PrismGlee/Meshes/M_MED_PrismGlee_CapeColliders
bodies {
 calf_l []
 calf_r []
 clavicle_l [{shapeType=Capsule, length=7.2736, radius=11.6327, center=<9.3571 -1.6271 -2.779>, rotation=<rot 69.3297 1.4476 176.9462>}]
 clavicle_r [
  {shapeType=Capsule, length=7.2736, radius=11.6327, center=<-9.3571 1.6271 2.779>, rotation=<rot 69.3297 1.4476 -3.0538>}
  {shapeType=Capsule, length=4.2271, radius=3.3224, center=<-14.7432 3.6227 -5.504>, rotation=<rot 73.6458 -17.5657 168.6697>}
 ]
 dyn_loincloth_rear_02 [{shapeType=Capsule, length=3.7965, radius=5.7688, shapeName=dyn_loincloth_rear_02_capsule, center=<5.4918 0 0>, rotation=<rot 90 0 0>}]
 dyn_loincloth_rear_02_l [{shapeType=Capsule, length=4.4036, radius=5.1571, shapeName=dyn_loincloth_rear_02_l_capsule, center=<5.7917 0 0>, rotation=<rot 90 180 0>}]
 dyn_loincloth_rear_02_r [{shapeType=Capsule, length=4.4036, radius=5.1571, shapeName=dyn_loincloth_rear_02_r_capsule, center=<-5.7917 0 0>, rotation=<rot 90 0 0>}]
 dyn_loincloth_rear_03 [{shapeType=Capsule, length=3.6815, radius=7.5721, shapeName=dyn_loincloth_rear_03_capsule, center=<11.0353 -1.5009 0>, rotation=<rot 90 0 0>}]
 dyn_loincloth_rear_03_l [{shapeType=Capsule, length=8.897, radius=6.6892, shapeName=dyn_loincloth_rear_03_l_capsule, center=<10.39 -1.9834 0>, rotation=<rot 90 180 0>}]
 dyn_loincloth_rear_03_r [{shapeType=Capsule, length=8.897, radius=6.6892, shapeName=dyn_loincloth_rear_03_r_capsule, center=<-10.39 1.9834 0>, rotation=<rot 90 0 0>}]
 dyn_shoulder_straps_03_l [{
   shapeType=Capsule
   length=8.0758
   radius=4.8837
   shapeName=dyn_shoulder_straps_03_l_capsule
   center=<2.1054 -2.7578 -5.7602>
   rotation=<rot -2.9845 158.4096 103.236>
  }]
 dyn_shoulder_straps_03_r [{
   shapeType=Capsule
   length=8.0758
   radius=4.8837
   shapeName=dyn_shoulder_straps_03_r_capsule
   center=<-2.1054 2.7578 5.7602>
   rotation=<rot -2.9845 158.4096 -76.764>
  }]
 foot_l []
 foot_r []
 lowerarm_l []
 lowerarm_r []
 neck_02 [{shapeType=Capsule, length=13.4375, radius=7.3175, center=<-0.2642 -0.3408 0>, rotation=<rot 90 -171.1193 -175.3978>}]
 pelvis [{shapeType=Capsule, length=3.4049, radius=13.9955, center=<-2.4887 0.1881 0>}]
 spine_01 [{shapeType=Capsule, length=7.9438, radius=13.5586, center=<-1.7965 -1.0072 0>}]
 spine_02 [
  {shapeType=Capsule, length=7.3264, radius=13.6322, center=<-1.711 -0.1851 0.0001>}
  {shapeType=Capsule, length=29.3956, radius=11.5807, center=<2.5024 0.4992 3.9825>, rotation=<rot 90 0.0049 0.0049>}
  {shapeType=Capsule, length=29.875, radius=11.1013, center=<2.4704 0.4991 -3.9825>, rotation=<rot 90 -0.0396 -0.0396>}
 ]
 spine_03 [{shapeType=Capsule, length=8.2107, radius=13.7031, center=<0.1445 0.2396 0.0001>}]
 spine_04 [{shapeType=Capsule, length=9.4404, radius=14.1325, center=<2.766 -1.2258 0.0002>}]
 spine_05 [
  {shapeType=Capsule, length=7.7665, radius=15.9511, center=<-0.7099 0.2093 0>}
  {shapeType=Capsule, length=13.9518, radius=13.9425, center=<-0.2916 0.4452 0>}
 ]
 thigh_l []
 thigh_r []
 upperarm_l []
 upperarm_r []
}
constraints []
